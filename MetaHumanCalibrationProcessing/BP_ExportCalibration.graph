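# Editor utility blueprint that exports selected Camera Calibration assets to a chosen directory for MetaHuman calibration processing.
# Its ExportCalibration function loops over the Content Browser selection, casts each asset to CameraCalibration and calls the native export with the Directory input.
# Names of successfully exported assets collect in an array, which is joined into a completion dialog when the loop finishes.
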EditorUtilityBlueprint /MetaHumanCalibrationProcessing/BP_ExportCalibration
graphs [
 EventGraph {links [], nodes []}
 ExportCalibration {nodes [
   n1 = ExportCalibration!() -> (then, Directory:Directory Path Structure) {is=title K2Node_FunctionEntry}
   > n2 = GetSelectedAssets()
   > n3 = For Each Loop(Array:Array of Object References=@n2) -> (Array Element:Object Reference)
   > n4 = Cast To CameraCalibration!(execute, Object=@n3) -> (then, CastFailed, AsCamera Calibration:Camera Calibration Object Reference) {is=title K2Node_DynamicCast}
   > n5 = ExportCalibration!(execute, InCameraCalibration:Camera Calibration Object Reference=@n4, InPath:"Directory Path Structure (by ref)"=@n1) -> (then, ReturnValue:bool) {is=title K2Node_CallFunction}
   > Branch(Condition=@n5)
   n6 = Get ObjectNames() -> (ObjectNames:Array of Strings)
   > Add(TargetArray=@n6, NewItem=GetObjectName(Object=@n3)) {is=title K2Node_CallArrayFunction}
   n7 = Make Array!(
    "[0]":String="To String (Text)"(InText=MakeLiteralText(Value="LOCTEXT(\"Finished exporting calibrations for:\")"))
    "[1]":String=JoinStringArray(SourceArray=@n6, Separator=", ")
   ) -> (Array:Array of Strings)
   Show Message Dialog(
    execute=@n3.Completed
    Title=MakeLiteralText(Value="LOCTEXT(\"Export finished\")")
    Message="To Text (String)"(InString=JoinStringArray(SourceArray=@n7, Separator=" "))
    MessageType=Ok
    DefaultValue=No
    MessageCategory=Success
   )
  ]}
]
variables []
